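# Physics asset for the Winter Hunter backpack cosmetic, defining how its dangling accessories swing.
# Capsule and box bodies cover the pack, bag, four small bag pieces, bead string, chain, coin, two diamonds and a two-segment skull string.
# Most pieces are constrained directly to the pack with limited swing and tighter or equal twist, while the bead string hangs off the bag and the skull string's second segment off its first.
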PhysicsAsset /BRCosmetics/Accessories/FORT_Backpacks/Backpack_F_MED_Winter_Hunter/Meshes/F_MED_Winter_Hunter_Physics
bodies {
 dyn_bag [{shapeType=Capsule, length=5.9406, radius=9.5227, center=<4.6725 -0.5248 -2.0549>, rotation=<rot 0 -178.6019 -89.7745>}]
 dyn_bag_bk_l [{shapeType=Capsule, length=2.6368, radius=1.3615, center=<0.3 0 0>}]
 dyn_bag_bk_r [{shapeType=Capsule, length=2.6368, radius=1.3615, center=<0.3 0 0>}]
 dyn_bag_fr_l [{shapeType=Capsule, length=2.6368, radius=1.3615, center=<0.3 0 0>}]
 dyn_bag_fr_r [{shapeType=Capsule, length=2.6368, radius=1.3615, center=<0.3 0 0>}]
 dyn_bead_string [{shapeType=Capsule, length=2, radius=2, center=<0.1 0 0>}]
 dyn_chain [{shapeType=Capsule, length=4.1396, radius=3.1, center=<6 0 0>, rotation=<rot -2.4356 -90 1.4519>}]
 dyn_coin [{shapeType=Box, extentX=1.2804, extentY=5.5861, extentZ=5.5523, center=<0 0.0817 0>}]
 dyn_diamond_bk [{shapeType=Capsule, length=0.1103, radius=1.6907, rotation=<rot 90 0 0>}]
 dyn_diamond_fr [{shapeType=Capsule, length=0.5604, radius=2.337, center=<-0.3488 0.0355 0.0076>, rotation=<rot 90 0 0>}]
 dyn_skull_string [{shapeType=Capsule, length=7.8894, radius=1, center=<7 -0.3766 0>}]
 dyn_skull_string1 [{shapeType=Box, extentX=6.1398, extentY=2.775, extentZ=9.1554, center=<3.6118 0.1288 0.0001>, rotation=<rot 0 5.6191 -0.0001>}]
 pack [{shapeType=Capsule, length=0.0001, radius=8.1093, center=<-0.2021 -7.5834 -6.6059>, rotation=<rot 76.2543 0 0>}]
}
constraints [
 {bone1Name=dyn_coin, bone2Name=pack, swing1LimitDegrees=5, swing2LimitDegrees=5, twistLimitDegrees=5}
 {bone1Name=dyn_diamond_fr, bone2Name=pack, swing1LimitDegrees=30, swing2LimitDegrees=30, twistLimitDegrees=5}
 {bone1Name=dyn_diamond_bk, bone2Name=pack, swing1LimitDegrees=30, swing2LimitDegrees=30, twistLimitDegrees=5}
 {bone1Name=dyn_skull_string, bone2Name=pack, swing1LimitDegrees=25, swing2LimitDegrees=25, twistLimitDegrees=15}
 {bone1Name=dyn_bag_bk_r, bone2Name=pack, swing1LimitDegrees=35, swing2LimitDegrees=35, twistLimitDegrees=10}
 {bone1Name=dyn_bag_bk_l, bone2Name=pack, swing1LimitDegrees=35, swing2LimitDegrees=35, twistLimitDegrees=10}
 {bone1Name=dyn_bag_fr_l, bone2Name=pack, swing1LimitDegrees=30, swing2LimitDegrees=30, twistLimitDegrees=10}
 {bone1Name=dyn_bag_fr_r, bone2Name=pack, swing1LimitDegrees=35, swing2LimitDegrees=35, twistLimitDegrees=10}
 {bone1Name=dyn_skull_string1, bone2Name=dyn_skull_string, swing1LimitDegrees=30, swing2LimitDegrees=35, twistLimitDegrees=5}
 {bone1Name=dyn_chain, bone2Name=pack, swing1LimitDegrees=10, swing2LimitDegrees=10, twistLimitDegrees=5}
 {bone1Name=dyn_bag, bone2Name=pack, swing1LimitDegrees=45, swing2LimitDegrees=45, twistLimitDegrees=5}
 {bone1Name=dyn_bead_string, bone2Name=dyn_bag, swing1LimitDegrees=5, swing2LimitDegrees=5, twistLimitDegrees=2}
]
